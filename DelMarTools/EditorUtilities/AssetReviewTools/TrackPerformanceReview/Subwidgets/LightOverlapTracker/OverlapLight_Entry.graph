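# Editor Utility Widget row for the track performance review's light overlap tracker, representing one actor in an overlap list.
# On Construct it shows the display name of its OverlappingActor variable as the label of its button.
# Clicking the button selects that actor in the level via EditorActorSubsystem, then runs the CAMERA ALIGN ACTIVEVIEWPORTONLY console command to frame it in the active viewport.
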EditorUtilityWidgetBlueprint /DelMarTools/EditorUtilities/AssetReviewTools/TrackPerformanceReview/Subwidgets/LightOverlapTracker/OverlapLight_Entry
graphs [EventGraph {nodes [
   "On Clicked (Button_0)"() {is=title K2Node_ComponentBoundEvent}
   > SetSelectedLevelActors(
    self=Get EditorActorSubsystem()
    ActorsToSelect=Make Array("[0]":Actor Object Reference=Get OverlappingActor() -> (OverlappingActor:Actor Object Reference)) -> (Array:Array of Actor Object References)
   )
   > ExecuteConsoleCommand(Command=CAMERA ALIGN ACTIVEVIEWPORTONLY)
   Event Construct()
   n1 = Get OverlappingActor()
   > "SetText (Text)"(self=Get Text_Name() -> (Text_Name:Text Object Reference), InText="To Text (String)"(InString=GetDisplayName(Object=@n1)))
   Event PreConstruct()
   Event Tick()
  ]}]
variables [OverlappingActor {type=Actor, default=null}]
widgets {
 info {rootWidgetClass=&StackBox, widgetCount=3}
 widgets [
  StackBox_0 {widgetClassPath=&StackBox, widget=StackBox_0}
  Button_0 {
   widgetClassPath=&Button
   parent=StackBox_0
   widget=Button_0
   properties {
    backgroundColor=<rgba 0.0469 0.0469 0.0469 0.3>
    widgetStyle {
     clickedSlateSound {resourceObject=null}
     disabled {
      bIsDynamicallyLoaded=false
      drawAs=RoundedBox
      imageType=NoImage
      mirroring=NoMirror
      resourceName=None
      resourceObject=null
      tiling=NoTile
      imageSize=<0 0>
      margin=<ltrb 0 0 0 0>
      outlineSettings {
       bUseBrushTransparency=false
       roundingType=FixedRadius
       width=1
       color {colorUseRule=UseColor_ColorTable, specifiedColor=<rgba 0 0 0 0>}
       cornerRadii=<4 4 4 4>
      }
      tintColor {colorUseRule=UseColor_Specified, specifiedColor=<rgba 0.0395 0.0395 0.0395 1>}
      uVRegion {bIsValid=false, max=<0 0>, min=<0 0>}
     }
     disabledForeground {colorUseRule=UseColor_ColorTable, specifiedColor=<rgba 0 0 0 0>}
     hovered {
      bIsDynamicallyLoaded=false
      drawAs=RoundedBox
      imageType=NoImage
      mirroring=NoMirror
      resourceName=None
      resourceObject=null
      tiling=NoTile
      imageSize=<32 32>
      margin=<ltrb 0 0 0 0>
      outlineSettings {
       bUseBrushTransparency=false
       roundingType=FixedRadius
       width=1
       color {colorUseRule=UseColor_ColorTable, specifiedColor=<rgba 0 0 0 0>}
       cornerRadii=<4 4 4 4>
      }
      tintColor {colorUseRule=UseColor_Specified, specifiedColor=<rgba 0.0953 0.0953 0.0953 1>}
      uVRegion {bIsValid=false, max=<0 0>, min=<0 0>}
     }
     hoveredForeground {colorUseRule=UseColor_ColorTable, specifiedColor=<rgba 0 0 0 0>}
     hoveredSlateSound {resourceObject=null}
     normal {
      bIsDynamicallyLoaded=false
      drawAs=RoundedBox
      imageType=NoImage
      mirroring=NoMirror
      resourceName=None
      resourceObject=null
      tiling=NoTile
      imageSize=<32 32>
      margin=<ltrb 0 0 0 0>
      outlineSettings {
       bUseBrushTransparency=false
       roundingType=FixedRadius
       width=1
       color {colorUseRule=UseColor_ColorTable, specifiedColor=<rgba 0 0 0 0>}
       cornerRadii=<4 4 4 4>
      }
      tintColor {colorUseRule=UseColor_Specified, specifiedColor=<rgba 0.0395 0.0395 0.0395 1>}
      uVRegion {bIsValid=false, max=<0 0>, min=<0 0>}
     }
     normalForeground {colorUseRule=UseColor_ColorTable, specifiedColor=<rgba 0 0 0 0>}
     normalPadding=<ltrb 12 1.5 12 1.5>
     pressed {
      bIsDynamicallyLoaded=false
      drawAs=RoundedBox
      imageType=NoImage
      mirroring=NoMirror
      resourceName=None
      resourceObject=null
      tiling=NoTile
      imageSize=<32 32>
      margin=<ltrb 0 0 0 0>
      outlineSettings {
       bUseBrushTransparency=false
       roundingType=FixedRadius
       width=1
       color {colorUseRule=UseColor_ColorTable, specifiedColor=<rgba 0 0 0 0>}
       cornerRadii=<4 4 4 4>
      }
      tintColor {colorUseRule=UseColor_Specified, specifiedColor=<rgba 0.0284 0.0284 0.0284 1>}
      uVRegion {bIsValid=false, max=<0 0>, min=<0 0>}
     }
     pressedForeground {colorUseRule=UseColor_ColorTable, specifiedColor=<rgba 0 0 0 0>}
     pressedPadding=<ltrb 12 2.5 12 0.5>
     pressedSlateSound {resourceObject=null}
    }
   }
   slot {class=StackBoxSlot}
  }
  Text_Name {
   widgetClassPath=&TextBlock
   bIsVariable=true
   parent=Button_0
   widget=Text_Name
   properties {
    text=Tree text
    font {
     bForceMonospaced=false
     bMaterialIsStencil=false
     fontMaterial=null
     fontObject=&Roboto
     letterSpacing=0
     monospacedWidth=1
     size=9
     skewAmount=0
     typefaceFontName=Regular
     outlineSettings {
      bApplyOutlineToDropShadows=false
      bMiteredCorners=false
      bSeparateFillAlpha=false
      outlineMaterial=null
      outlineSize=0
      outlineColor=<rgba 0 0 0 1>
     }
    }
   }
   slot {class=ButtonSlot, properties {padding=<ltrb 10 2 10 2>}}
  }
 ]
}
