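# Editor utility row widget for the light overlap tool in track performance review, showing one light actor's overlap data.
# On construct it shows the actor's name, light component count and class, and mobility.
# SetTrackMinDistance shows track distance; AddOverlap collects unique overlapping owner actors, updates the count, reveals hidden controls and adds an OverlapLight_Entry child to an expandable list.
# Buttons select the light or its overlaps and align the viewport camera.
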EditorUtilityWidgetBlueprint /DelMarTools/EditorUtilities/AssetReviewTools/TrackPerformanceReview/Subwidgets/LightOverlapTracker/LightOverlapTool_Entry
graphs [
 AddOverlap {nodes [
   n1 = AddOverlap!() -> (then, Overlapping:Light Component Object Reference) {is=title K2Node_FunctionEntry}
   n2 = Get OverlappingLightActors() -> (OverlappingLightActors:Array of Actor Object References)
   > Add Unique(TargetArray:Array of Actor Object References=@n2, NewItem:"Actor Object Reference (by ref)"=GetOwner(self=@n1))
   > n3 = Set OverlappingAmount(OverlappingAmount:int=Length(TargetArray=@n2)) -> (Output_Get:int=0)
   > "SetText (Text)"(
    self=Get Text_AmountOfOverlaps() -> (Text_AmountOfOverlaps:Text Object Reference)
    InText=Format Text!(Format="LOCTEXT(\"{num}\")", num:int=@n3) -> (Result)
   )
   > SetVisibility(
    self=[
     Get Button_SelectOverlappingLights() -> (Button_SelectOverlappingLights:Editor Utility Button Object Reference)
     Get EditorUtilityExpandableArea_0() -> (EditorUtilityExpandableArea_0:Editor Utility Expandable Area Object Reference)
    ]
    InVisibility=Visible
   )
   > n4 = Create Widget!(
    execute:Exec
    WorldContextObject:Object Reference
    Class:User Widget Class Reference=&/OverlapLight_Entry.OverlapLight_Entry_C
    OwningPlayer:Player Controller Object Reference
    OverlappingActor:Actor Object Reference=GetOwner(self=Get Overlapping() -> (Overlapping:Light Component Object Reference))
   ) -> (then:Exec, ReturnValue:Overlap Light Entry Object Reference)
   > AddChild(self=Get VerticalBox_Overlaps() -> (VerticalBox_Overlaps:Vertical Box Object Reference), Content=@n4)
  ]}
 EventGraph {nodes [
   "On Clicked (Button_SelectLights)"() {is=title K2Node_ComponentBoundEvent}
   > n1 = SetSelectedLevelActors(
    self=Get EditorActorSubsystem()
    ActorsToSelect=Make Array("[0]":Actor Object Reference=Get LightOwningActor() -> (LightOwningActor:Actor Object Reference)) -> (Array:Array of Actor Object References)
   )
   ExecuteConsoleCommand(execute=[@n1.then, @n2.then], Command=CAMERA ALIGN ACTIVEVIEWPORTONLY)
   "On Clicked (Button_SelectOverlappingLights)"() {is=title K2Node_ComponentBoundEvent}
   > n2 = SetSelectedLevelActors(self=Get EditorActorSubsystem(), ActorsToSelect=Get OverlappingLightActors())
   Event Construct()
   n3 = Get LightOwningActor()
   > "SetText (Text)"(self=Get Text_ActorName() -> (Text_ActorName:Text Object Reference), InText="To Text (String)"(InString=GetDisplayName(Object=@n3)))
   n4 = Get Components By Class!(self:Actor Object Reference=Get LightOwningActor(), ComponentClass:Actor Component Class Reference=/Script/Engine.LightComponent) -> (ReturnValue:Array of Light Component Object References)
   n5 = Length(TargetArray:Array of Light Component Object References=@n4)
   n6 = "Get (a copy)"(Array:Array of Light Component Object References=@n4) -> (Output:Light Component Object Reference)
   n7 = Get Class(Object=@n6) -> (ReturnValue:Light Component Class Reference)
   n8 = "integer > integer"(A=@n5, B=1)
   > "SetText (Text)"(
    self=Get Text_AmountOfComponents() -> (Text_AmountOfComponents:Text Object Reference)
    InText=Format Text!(
     Format="LOCTEXT(\"{amount} {componentName}{s}\")"
     amount:int=@n5
     componentName:String=Get Class Display Name(Class=@n7)
     s:String=SelectString(A=s, bPickA=@n8)
    ) -> (Result)
   )
   n9 = Get Mobility!(self:Scene Component Object Reference=@n6) -> (Mobility:EComponentMobility Enum=Static)
   > "SetText (Text)"(
    self=Get Text_Mobility() -> (Text_Mobility:Text Object Reference)
    InText="To Text (String)"(InString=Enum to String!(Enumerator:EComponentMobility Enum=@n9) -> (ReturnValue:String))
   )
  ]}
 SetTrackMinDistance {nodes [
   n1 = SetTrackMinDistance!() -> (then, MinDistance:double=~) {is=title K2Node_FunctionEntry}
   > "SetText (Text)"(
    self=Get Text_MinDistance() -> (Text_MinDistance:Text Object Reference)
    InText=Format Text!(Format="LOCTEXT(\"{Distance} \")", Distance:double=@n1) -> (Result)
   )
   > Return Node()
  ]}
]
variables [
 LightOwningActor {type=Actor, default=null}
 OverlappingAmount {type=integer}
 OverlappingLightActors {type="array<Actor>"}
]
widgets {
 info {namedSlotCount=2, rootWidgetClass=&VerticalBox, widgetCount=20}
 widgets [
  VerticalBox_86 {widgetClassPath=&VerticalBox, widget=VerticalBox_86}
  Border_71 {
   widgetClassPath=&Border
   parent=VerticalBox_86
   widget=Border_71
   properties {brushColor=<rgba 0.1458 0.1458 0.1458 1>, padding=<ltrb 1 1 1 1>, renderTransform {angle=0, scale=<1 0.5>, shear=<0 0>, translation=<0 0>}}
   slot {class=VerticalBoxSlot, properties {padding=<ltrb 10 10 10 0>}}
  }
  StackBox_0 {widgetClassPath=&StackBox, parent=VerticalBox_86, widget=StackBox_0, slot {class=VerticalBoxSlot}}
  Button_SelectLights {
   widgetClassPath=&EditorUtilityButton
   parent=StackBox_0
   widget=Button_SelectLights
   properties {toolTipText=Select Actor}
   slot {class=StackBoxSlot, properties {padding=<ltrb 10 0 0 0>}}
  }
  Image {
   widgetClassPath=&Image
   parent=Button_SelectLights
   widget=Image
   properties {brush {
     bIsDynamicallyLoaded=false
     drawAs=Image
     imageType=FullColor
     mirroring=NoMirror
     resourceName=None
     resourceObject=/Game/UI/Friends_UI/Social/Search
     tiling=NoTile
     imageSize=<20 20>
     margin=<ltrb 0 0 0 0>
     outlineSettings {
      bUseBrushTransparency=false
      roundingType=HalfHeightRadius
      width=0
      color {colorUseRule=UseColor_Specified, specifiedColor=<rgba 0 0 0 0>}
      cornerRadii=<0 0 0 0>
     }
     tintColor {colorUseRule=UseColor_Specified, specifiedColor=<rgba 1 1 1 1>}
     uVRegion {bIsValid=false, max=<0 0>, min=<0 0>}
    }}
   slot {class=ButtonSlot}
  }
  Text_ActorName {
   widgetClassPath=&TextBlock
   bIsVariable=true
   parent=StackBox_0
   widget=Text_ActorName
   properties {
    text="<<Actor Name>>"
    font {
     bForceMonospaced=false
     bMaterialIsStencil=false
     fontMaterial=null
     fontObject=&Roboto
     letterSpacing=0
     monospacedWidth=1
     size=9
     skewAmount=0
     typefaceFontName=Bold
     outlineSettings {
      bApplyOutlineToDropShadows=false
      bMiteredCorners=false
      bSeparateFillAlpha=false
      outlineMaterial=null
      outlineSize=0
      outlineColor=<rgba 0 0 0 1>
     }
    }
   }
   slot {
    class=StackBoxSlot
    properties {horizontalAlignment=HAlign_Left, verticalAlignment=VAlign_Center, padding=<ltrb 10 0 0 0>, size {sizeRule=Fill, value=1}}
   }
  }
  SizeBox_2 {
   widgetClassPath=&SizeBox
   parent=StackBox_0
   widget=SizeBox_2
   properties {bOverride_WidthOverride=true, widthOverride=130}
   slot {class=StackBoxSlot, properties {horizontalAlignment=HAlign_Left, verticalAlignment=VAlign_Center, padding=<ltrb 10 0 0 0>}}
  }
  Text_AmountOfComponents {
   widgetClassPath=&TextBlock
   bIsVariable=true
   parent=SizeBox_2
   widget=Text_AmountOfComponents
   properties {
    text="<<Component>>"
    font {
     bForceMonospaced=false
     bMaterialIsStencil=false
     fontMaterial=null
     fontObject=&Roboto
     letterSpacing=0
     monospacedWidth=1
     size=9
     skewAmount=0
     typefaceFontName=Regular
     outlineSettings {
      bApplyOutlineToDropShadows=false
      bMiteredCorners=false
      bSeparateFillAlpha=false
      outlineMaterial=null
      outlineSize=0
      outlineColor=<rgba 0 0 0 1>
     }
    }
   }
   slot {class=SizeBoxSlot}
  }
  SizeBox_3 {
   widgetClassPath=&SizeBox
   parent=StackBox_0
   widget=SizeBox_3
   properties {bOverride_WidthOverride=true, widthOverride=60}
   slot {class=StackBoxSlot, properties {horizontalAlignment=HAlign_Left, verticalAlignment=VAlign_Center, padding=<ltrb 10 0 0 0>}}
  }
  Text_Mobility {
   widgetClassPath=&TextBlock
   bIsVariable=true
   parent=SizeBox_3
   widget=Text_Mobility
   properties {
    text="<<Mobility>>"
    font {
     bForceMonospaced=false
     bMaterialIsStencil=false
     fontMaterial=null
     fontObject=&Roboto
     letterSpacing=0
     monospacedWidth=1
     size=9
     skewAmount=0
     typefaceFontName=Regular
     outlineSettings {
      bApplyOutlineToDropShadows=false
      bMiteredCorners=false
      bSeparateFillAlpha=false
      outlineMaterial=null
      outlineSize=0
      outlineColor=<rgba 0 0 0 1>
     }
    }
   }
   slot {class=SizeBoxSlot}
  }
  SizeBox_0 {
   widgetClassPath=&SizeBox
   parent=StackBox_0
   widget=SizeBox_0
   properties {bOverride_WidthOverride=true, widthOverride=110}
   slot {class=StackBoxSlot, properties {horizontalAlignment=HAlign_Left, verticalAlignment=VAlign_Center, padding=<ltrb 10 0 0 0>}}
  }
  Text_MinDistance {
   widgetClassPath=&TextBlock
   bIsVariable=true
   parent=SizeBox_0
   widget=Text_MinDistance
   properties {
    text=DistanceFromTrack
    font {
     bForceMonospaced=false
     bMaterialIsStencil=false
     fontMaterial=null
     fontObject=&Roboto
     letterSpacing=0
     monospacedWidth=1
     size=9
     skewAmount=0
     typefaceFontName=Regular
     outlineSettings {
      bApplyOutlineToDropShadows=false
      bMiteredCorners=false
      bSeparateFillAlpha=false
      outlineMaterial=null
      outlineSize=0
      outlineColor=<rgba 0 0 0 1>
     }
    }
   }
   slot {class=SizeBoxSlot}
  }
  SizeBox_1 {
   widgetClassPath=&SizeBox
   parent=StackBox_0
   widget=SizeBox_1
   properties {bOverride_WidthOverride=true, widthOverride=110}
   slot {class=StackBoxSlot, properties {horizontalAlignment=HAlign_Left, verticalAlignment=VAlign_Center, padding=<ltrb 10 0 0 0>}}
  }
  Text_AmountOfOverlaps {
   widgetClassPath=&TextBlock
   bIsVariable=true
   parent=SizeBox_1
   widget=Text_AmountOfOverlaps
   properties {
    text=0
    font {
     bForceMonospaced=false
     bMaterialIsStencil=false
     fontMaterial=null
     fontObject=&Roboto
     letterSpacing=0
     monospacedWidth=1
     size=9
     skewAmount=0
     typefaceFontName=Regular
     outlineSettings {
      bApplyOutlineToDropShadows=false
      bMiteredCorners=false
      bSeparateFillAlpha=false
      outlineMaterial=null
      outlineSize=0
      outlineColor=<rgba 0 0 0 1>
     }
    }
   }
   slot {class=SizeBoxSlot}
  }
  Button_SelectOverlappingLights {
   widgetClassPath=&EditorUtilityButton
   parent=StackBox_0
   widget=Button_SelectOverlappingLights
   properties {toolTipText=Select Light Overlapping Actors, visibility=Hidden}
   slot {class=StackBoxSlot, properties {padding=<ltrb 10 0 0 0>}}
  }
  Image_1 {
   widgetClassPath=&Image
   parent=Button_SelectOverlappingLights
   widget=Image_1
   properties {brush {
     bIsDynamicallyLoaded=false
     drawAs=Image
     imageType=FullColor
     mirroring=NoMirror
     resourceName=None
     resourceObject=&Search
     tiling=NoTile
     imageSize=<20 20>
     margin=<ltrb 0 0 0 0>
     outlineSettings {
      bUseBrushTransparency=false
      roundingType=HalfHeightRadius
      width=0
      color {colorUseRule=UseColor_Specified, specifiedColor=<rgba 0 0 0 0>}
      cornerRadii=<0 0 0 0>
     }
     tintColor {colorUseRule=UseColor_Specified, specifiedColor=<rgba 1 1 1 1>}
     uVRegion {bIsValid=false, max=<0 0>, min=<0 0>}
    }}
   slot {class=ButtonSlot}
  }
  EditorUtilityExpandableArea_0 {
   widgetClassPath=/Script/Blutility.EditorUtilityExpandableArea
   bIsVariable=true
   parent=StackBox_0
   widget=EditorUtilityExpandableArea_0
   properties {visibility=Hidden}
   slot {class=StackBoxSlot, properties {padding=<ltrb 10 0 0 0>, size {sizeRule=Fill, value=1}}}
  }
  Text_AmountOfOverlaps_1 {
   widgetClassPath=&TextBlock
   bIsVariable=true
   namedSlotHost=EditorUtilityExpandableArea_0
   widget=Text_AmountOfOverlaps_1
   properties {
    text=Overlaps
    font {
     bForceMonospaced=false
     bMaterialIsStencil=false
     fontMaterial=null
     fontObject=&Roboto
     letterSpacing=0
     monospacedWidth=1
     size=9
     skewAmount=0
     typefaceFontName=Regular
     outlineSettings {
      bApplyOutlineToDropShadows=false
      bMiteredCorners=false
      bSeparateFillAlpha=false
      outlineMaterial=null
      outlineSize=0
      outlineColor=<rgba 0 0 0 1>
     }
    }
   }
  }
  VerticalBox_Overlaps {widgetClassPath=&VerticalBox, bIsVariable=true, namedSlotHost=EditorUtilityExpandableArea_0, widget=VerticalBox_Overlaps}
  HorizontalBox_0 {widgetClassPath=&HorizontalBox, parent=VerticalBox_86, widget=HorizontalBox_0, slot {class=VerticalBoxSlot}}
 ]
}
