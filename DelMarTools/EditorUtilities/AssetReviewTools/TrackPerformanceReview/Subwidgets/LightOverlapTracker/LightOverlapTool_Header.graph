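# Editor utility widget that serves as the column header row for the Light Overlap tool in the DelMar track performance review tools.
# A StackBox lays out bold labels for actor name, component count, mobility, distance from track, overlap amount and overlaps, using fixed-width SizeBoxes and fill slots.
# Two select buttons are Hidden rather than Collapsed, so they still reserve column space, likely to align with data rows.
# The event graph is empty.
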EditorUtilityWidgetBlueprint /DelMarTools/EditorUtilities/AssetReviewTools/TrackPerformanceReview/Subwidgets/LightOverlapTracker/LightOverlapTool_Header
graphs [EventGraph {links [], nodes []}]
variables []
widgets {
 info {rootWidgetClass=&VerticalBox, widgetCount=17}
 widgets [
  VerticalBox_86 {widgetClassPath=&VerticalBox, widget=VerticalBox_86}
  StackBox_0 {widgetClassPath=&StackBox, parent=VerticalBox_86, widget=StackBox_0, slot {class=VerticalBoxSlot}}
  Button_SelectLights {
   widgetClassPath=&EditorUtilityButton
   parent=StackBox_0
   widget=Button_SelectLights
   properties {toolTipText=Select Actor, visibility=Hidden}
   slot {class=StackBoxSlot, properties {padding=<ltrb 10 0 0 0>}}
  }
  Image {
   widgetClassPath=&Image
   parent=Button_SelectLights
   widget=Image
   properties {brush {
     bIsDynamicallyLoaded=false
     drawAs=Image
     imageType=FullColor
     mirroring=NoMirror
     resourceName=None
     resourceObject=/Game/UI/Friends_UI/Social/Search
     tiling=NoTile
     imageSize=<20 20>
     margin=<ltrb 0 0 0 0>
     outlineSettings {
      bUseBrushTransparency=false
      roundingType=HalfHeightRadius
      width=0
      color {colorUseRule=UseColor_Specified, specifiedColor=<rgba 0 0 0 0>}
      cornerRadii=<0 0 0 0>
     }
     tintColor {colorUseRule=UseColor_Specified, specifiedColor=<rgba 1 1 1 1>}
     uVRegion {bIsValid=false, max=<0 0>, min=<0 0>}
    }}
   slot {class=ButtonSlot}
  }
  Text_ActorName {
   widgetClassPath=&TextBlock
   bIsVariable=true
   parent=StackBox_0
   widget=Text_ActorName
   properties {
    text=Actor Name
    font {
     bForceMonospaced=false
     bMaterialIsStencil=false
     fontMaterial=null
     fontObject=&Roboto
     letterSpacing=0
     monospacedWidth=1
     size=9
     skewAmount=0
     typefaceFontName=Bold
     outlineSettings {
      bApplyOutlineToDropShadows=false
      bMiteredCorners=false
      bSeparateFillAlpha=false
      outlineMaterial=null
      outlineSize=0
      outlineColor=<rgba 0 0 0 1>
     }
    }
   }
   slot {
    class=StackBoxSlot
    properties {horizontalAlignment=HAlign_Left, verticalAlignment=VAlign_Center, padding=<ltrb 10 0 0 0>, size {sizeRule=Fill, value=1}}
   }
  }
  SizeBox_2 {
   widgetClassPath=&SizeBox
   parent=StackBox_0
   widget=SizeBox_2
   properties {bOverride_WidthOverride=true, widthOverride=130}
   slot {class=StackBoxSlot, properties {horizontalAlignment=HAlign_Left, verticalAlignment=VAlign_Center, padding=<ltrb 10 0 0 0>}}
  }
  Text_AmountOfComponents {
   widgetClassPath=&TextBlock
   bIsVariable=true
   parent=SizeBox_2
   widget=Text_AmountOfComponents
   properties {
    text=Amount of Components
    font {
     bForceMonospaced=false
     bMaterialIsStencil=false
     fontMaterial=null
     fontObject=&Roboto
     letterSpacing=0
     monospacedWidth=1
     size=9
     skewAmount=0
     typefaceFontName=Bold
     outlineSettings {
      bApplyOutlineToDropShadows=false
      bMiteredCorners=false
      bSeparateFillAlpha=false
      outlineMaterial=null
      outlineSize=0
      outlineColor=<rgba 0 0 0 1>
     }
    }
   }
   slot {class=SizeBoxSlot}
  }
  SizeBox_3 {
   widgetClassPath=&SizeBox
   parent=StackBox_0
   widget=SizeBox_3
   properties {bOverride_WidthOverride=true, widthOverride=60}
   slot {class=StackBoxSlot, properties {horizontalAlignment=HAlign_Left, verticalAlignment=VAlign_Center, padding=<ltrb 10 0 0 0>}}
  }
  Text_Mobility {
   widgetClassPath=&TextBlock
   bIsVariable=true
   parent=SizeBox_3
   widget=Text_Mobility
   properties {
    text=Mobility
    font {
     bForceMonospaced=false
     bMaterialIsStencil=false
     fontMaterial=null
     fontObject=&Roboto
     letterSpacing=0
     monospacedWidth=1
     size=9
     skewAmount=0
     typefaceFontName=Bold
     outlineSettings {
      bApplyOutlineToDropShadows=false
      bMiteredCorners=false
      bSeparateFillAlpha=false
      outlineMaterial=null
      outlineSize=0
      outlineColor=<rgba 0 0 0 1>
     }
    }
   }
   slot {class=SizeBoxSlot}
  }
  SizeBox_1 {
   widgetClassPath=&SizeBox
   parent=StackBox_0
   widget=SizeBox_1
   properties {bOverride_WidthOverride=true, widthOverride=110}
   slot {class=StackBoxSlot, properties {horizontalAlignment=HAlign_Left, verticalAlignment=VAlign_Center, padding=<ltrb 10 0 0 0>}}
  }
  Text_MinDistance {
   widgetClassPath=&TextBlock
   bIsVariable=true
   parent=SizeBox_1
   widget=Text_MinDistance
   properties {
    text=DistanceFromTrack
    font {
     bForceMonospaced=false
     bMaterialIsStencil=false
     fontMaterial=null
     fontObject=&Roboto
     letterSpacing=0
     monospacedWidth=1
     size=9
     skewAmount=0
     typefaceFontName=Bold
     outlineSettings {
      bApplyOutlineToDropShadows=false
      bMiteredCorners=false
      bSeparateFillAlpha=false
      outlineMaterial=null
      outlineSize=0
      outlineColor=<rgba 0 0 0 1>
     }
    }
   }
   slot {class=SizeBoxSlot}
  }
  SizeBox_0 {
   widgetClassPath=&SizeBox
   parent=StackBox_0
   widget=SizeBox_0
   properties {bOverride_WidthOverride=true, widthOverride=110}
   slot {class=StackBoxSlot, properties {horizontalAlignment=HAlign_Left, verticalAlignment=VAlign_Center, padding=<ltrb 10 0 0 0>}}
  }
  Text_AmountOfOverlaps {
   widgetClassPath=&TextBlock
   bIsVariable=true
   parent=SizeBox_0
   widget=Text_AmountOfOverlaps
   properties {
    text=Overlap Amount
    font {
     bForceMonospaced=false
     bMaterialIsStencil=false
     fontMaterial=null
     fontObject=&Roboto
     letterSpacing=0
     monospacedWidth=1
     size=9
     skewAmount=0
     typefaceFontName=Bold
     outlineSettings {
      bApplyOutlineToDropShadows=false
      bMiteredCorners=false
      bSeparateFillAlpha=false
      outlineMaterial=null
      outlineSize=0
      outlineColor=<rgba 0 0 0 1>
     }
    }
   }
   slot {class=SizeBoxSlot}
  }
  Button_SelectOverlappingLights {
   widgetClassPath=&EditorUtilityButton
   parent=StackBox_0
   widget=Button_SelectOverlappingLights
   properties {toolTipText=Select Light Overlapping Actors, visibility=Hidden}
   slot {class=StackBoxSlot, properties {padding=<ltrb 10 0 0 0>}}
  }
  Image_1 {
   widgetClassPath=&Image
   parent=Button_SelectOverlappingLights
   widget=Image_1
   properties {brush {
     bIsDynamicallyLoaded=false
     drawAs=Image
     imageType=FullColor
     mirroring=NoMirror
     resourceName=None
     resourceObject=&Search
     tiling=NoTile
     imageSize=<20 20>
     margin=<ltrb 0 0 0 0>
     outlineSettings {
      bUseBrushTransparency=false
      roundingType=HalfHeightRadius
      width=0
      color {colorUseRule=UseColor_Specified, specifiedColor=<rgba 0 0 0 0>}
      cornerRadii=<0 0 0 0>
     }
     tintColor {colorUseRule=UseColor_Specified, specifiedColor=<rgba 1 1 1 1>}
     uVRegion {bIsValid=false, max=<0 0>, min=<0 0>}
    }}
   slot {class=ButtonSlot}
  }
  Text_AmountOfOverlaps_1 {
   widgetClassPath=&TextBlock
   bIsVariable=true
   parent=StackBox_0
   widget=Text_AmountOfOverlaps_1
   properties {
    text=Overlaps
    font {
     bForceMonospaced=false
     bMaterialIsStencil=false
     fontMaterial=null
     fontObject=&Roboto
     letterSpacing=0
     monospacedWidth=1
     size=9
     skewAmount=0
     typefaceFontName=Bold
     outlineSettings {
      bApplyOutlineToDropShadows=false
      bMiteredCorners=false
      bSeparateFillAlpha=false
      outlineMaterial=null
      outlineSize=0
      outlineColor=<rgba 0 0 0 1>
     }
    }
   }
   slot {
    class=StackBoxSlot
    properties {horizontalAlignment=HAlign_Left, verticalAlignment=VAlign_Center, padding=<ltrb 10 0 0 0>, size {sizeRule=Fill, value=1}}
   }
  }
  HorizontalBox_0 {widgetClassPath=&HorizontalBox, parent=VerticalBox_86, widget=HorizontalBox_0, slot {class=VerticalBoxSlot}}
 ]
}
