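# Editor utility widget for track performance review that lists level actors with lights, their distance to the nearest track spline and which lights overlap.
# Refresh rebuilds the entries, testing each actor's lights against those gathered so far and recording hits on both entries.
# Point pairs compare distance to summed radii.
# Spot lights use a temporary comparer actor, fitting cone meshes to angle and attenuation for component overlap queries against cones or a sphere.
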EditorUtilityWidgetBlueprint /DelMarTools/EditorUtilities/AssetReviewTools/TrackPerformanceReview/Subwidgets/LightOverlapTracker/LightOverlapTool
graphs [
 CheckOverlapPointToPoint {nodes [
   n1 = CheckOverlapPointToPoint!() -> (then, PointLightA:Point Light Component Object Reference, PointLightB:Point Light Component Object Reference) {is=title K2Node_FunctionEntry}
   n2 = Get World Location!(self:Scene Component Object Reference=@n1.PointLightB) -> (ReturnValue:Vector) {is=title K2Node_CallFunction}
   n3 = Get World Location(self=@n1)
   n4 = Get AttenuationRadius!(self:Local Light Component Object Reference=@n1.PointLightB) -> (AttenuationRadius:float)
   n5 = Get AttenuationRadius(self=@n1)
   > Return Node!(
    execute
    bOverlaps:bool="float < float"(A="Distance (Vector)"!(V1:Vector=@n3, V2:Vector=@n2) -> (ReturnValue:double), B=float + float(A=@n5, B=@n4))
   )
  ]}
 CheckOverlapSpotToPoint {nodes [
   CheckOverlapSpotToPoint!() -> (then, SpotLight:Spot Light Component Object Reference, PointLight:Point Light Component Object Reference) {is=title K2Node_FunctionEntry}
   > Set World Transform!(
    execute
    self:Scene Component Object Reference=Get ConeA!(self:EUBP Light Overlapping Comparer Object Reference=Get ComparerActor() -> (ComparerActor:EUBP Light Overlapping Comparer Object Reference)) -> (ConeA:Static Mesh Component Object Reference)
    NewTransform:"Transform (by ref)"=ConvertLightInfoToConeTransform!(SpotLight:Spot Light Component Object Reference=Get SpotLight() -> (SpotLight:Spot Light Component Object Reference)) -> (ConeTransform:Transform)
    bSweep:bool=false
    bTeleport:bool=false
   ) -> (then, SweepHitResult:Hit Result Structure) {is=title K2Node_CallFunction}
   > Set World Location!(
    execute
    self:Scene Component Object Reference=Get Sphere!(self:EUBP Light Overlapping Comparer Object Reference=Get ComparerActor()) -> (Sphere:Sphere Collision Object Reference)
    NewLocation:Vector=Get World Location(self=Get PointLight() -> (PointLight:Point Light Component Object Reference))
    bSweep:bool=false
    bTeleport:bool=false
   ) -> (then, SweepHitResult:Hit Result Structure) {is=title K2Node_CallFunction}
   > SetSphereRadius!(
    execute
    self:Sphere Collision Object Reference=Get Sphere(self=Get ComparerActor())
    InSphereRadius:float=Get AttenuationRadius(self=Get PointLight())
    bUpdateOverlaps:bool=true
   ) -> (then)
   n1 = Get ComparerActor()
   n2 = Get ConeA(self=@n1)
   > n3 = Component Overlap Components!(
    execute
    Component:Primitive Component Object Reference=@n2
    ComponentTransform:"Transform (by ref)"=Get World Transform!(self:Scene Component Object Reference=@n2) -> (ReturnValue:Transform) {is=title K2Node_CallFunction}
    ObjectTypes:Array of EObjectTypeQuery Enums=Make Array("[0]":EObjectTypeQuery Enum=ObjectTypeQuery2) -> (Array:Array of EObjectTypeQuery Enums)
    ComponentClassFilter:Object Class Reference
    ActorsToIgnore:Array of Actor Object References
   ) -> (then, OutComponents:Array of Primitive Component Object References, ReturnValue:bool)
   > Return Node!(
    execute
    overlaps:bool=Contains Item(
     TargetArray:Array of Primitive Component Object References=@n3
     ItemToFind:"Primitive Component Object Reference (by ref)"=Get Sphere(self=@n1)
    )
   )
  ]}
 CheckOverlapSpotToSpot {nodes [
   CheckOverlapSpotToSpot!() -> (then, SpotLightA:Spot Light Component Object Reference, SpotLightB:Spot Light Component Object Reference) {is=title K2Node_FunctionEntry}
   > Set World Transform(
    self=Get ConeA(self=Get ComparerActor())
    NewTransform=ConvertLightInfoToConeTransform(SpotLight=Get SpotLightA() -> (SpotLightA:Spot Light Component Object Reference))
    bSweep=false
    bTeleport=false
   )
   > Set World Transform(
    self=Get ConeB!(self:EUBP Light Overlapping Comparer Object Reference=Get ComparerActor()) -> (ConeB:Static Mesh Component Object Reference)
    NewTransform=ConvertLightInfoToConeTransform(SpotLight=Get SpotLightB() -> (SpotLightB:Spot Light Component Object Reference))
    bSweep=false
    bTeleport=false
   )
   n1 = Get ComparerActor()
   n2 = Get ConeA(self=@n1)
   > n3 = Component Overlap Components(Component=@n2, ComponentTransform=Get World Transform(self=@n2), ObjectTypes=Make Array("[0]"=ObjectTypeQuery2))
   > Return Node!(execute, bOverlaps:bool=Contains Item(TargetArray=@n3, ItemToFind=Get ConeB(self=@n1)))
  ]}
 ConvertLightInfoToConeTransform {nodes [
   ConvertLightInfoToConeTransform!() -> (then, SpotLight:Spot Light Component Object Reference) {is=title K2Node_FunctionEntry}
   n1 = Get SpotLight()
   n2 = Get SpotLight()
   n3 = float * float(
    A=Get AttenuationRadius(self=@n2)
    B="Cos (Degrees)"!(A:double=Get OuterConeAngle!(self:Spot Light Component Object Reference=@n2) -> (OuterConeAngle:float)) -> (ReturnValue:double)
   ) {is=title K2Node_PromotableOperator}
   n4 = float / float(A=@n3, B=100)
   n5 = float * float!(A="Tan (Degrees)"!(A:double=Get OuterConeAngle(self=@n1)) -> (ReturnValue:double), B=2, C:double=@n4) -> (ReturnValue)
   n6 = vector * vector(A=GetForwardVector!(self:Scene Component Object Reference=Get SpotLight()) -> (ReturnValue:Vector), B:float=-1) -> (ReturnValue=~)
   n7 = Get World Location(self=Get SpotLight())
   n8 = vector * vector(A=GetForwardVector(self=Get SpotLight()), B:double=float / float(A=@n3, B=2)) -> (ReturnValue=~)
   > Return Node!(
    execute
    ConeTransform:Transform=MakeTransform(
     Location=vector + vector(A=@n7, B=@n8)
     Rotation=MakeRotFromZ!(Z:"Vector (by ref)"=@n6) -> (ReturnValue:Rotator)
     Scale=MakeVector(X=@n5, Y=@n5, Z=@n4)
    )
   )
  ]}
 CreateEntry {nodes [
   CreateEntry!() -> (then, LightActor:Actor Object Reference) {is=title K2Node_FunctionEntry}
   > n1 = Sequence!(execute) -> (then_0, then_1, then_2:Exec)
   > n2 = Create Widget!(
    execute:Exec
    WorldContextObject:Object Reference
    Class:User Widget Class Reference=&/LightOverlapTool_Entry.LightOverlapTool_Entry_C
    OwningPlayer:Player Controller Object Reference
    LightOwningActor:Actor Object Reference=Get LightActor() -> (LightActor:Actor Object Reference)
   ) -> (then:Exec, ReturnValue:Light Overlap Tool Entry Object Reference)
   > n3 = Set EntryWidget(EntryWidget:Light Overlap Tool Entry Object Reference=@n2) -> (Output_Get:Light Overlap Tool Entry Object Reference)
   > AddChild(self=Get EditorUtilityScrollBox_Entries() -> (EditorUtilityScrollBox_Entries:Editor Utility Scroll Box Object Reference), Content=@n3)
   > Add!(
    execute
    TargetMap:Map of Actor Object References to Light Overlap Tool Entry Object References=Get LightActorsEntries() -> (LightActorsEntries:Map of Actor Object References to Light Overlap Tool Entry Object References)
    Key:"Actor Object Reference (by ref)"=Get LightActor()
    Value:"Light Overlap Tool Entry Object Reference (by ref)"=Get EntryWidget() -> (EntryWidget:Light Overlap Tool Entry Object Reference)
   ) -> (then) {is=title K2Node_CallFunction}
   n4 = For Each Loop(Exec=@n1.then_1, Array:Array of Object References=Get TrackActors() -> (TrackActors:Array of Object References)) -> (Array Element:Object Reference)
   n5 = Get Actor Location(self=Get LightActor())
   n6 = "Distance (Vector)"(
    V1=FindLocationClosestToWorldLocation!(
     self:Spline Component Object Reference=Get Spline!(self:Object Reference=@n4) -> (Spline:Object Reference)
     WorldLocation:"Vector (by ref)"=@n5
     CoordinateSpace:ESplineCoordinateSpace Enum=World
    ) -> (ReturnValue:Vector)
    V2=@n5
   )
   > Branch(Condition="float < float"(A=@n6, B=Get MinDistance() -> (MinDistance:double)))
   > Set MinDistance(MinDistance:double=@n6) -> (Output_Get:double)
   n7 = Get MinDistance()
   SetTrackMinDistance!(
    execute=@n4.Completed
    self:Light Overlap Tool Entry Object Reference=Get EntryWidget()
    MinDistance:double=SelectFloat(A=0, B=@n7, bPickA="Equal (Float)"(A=@n7, B=500000000))
   ) -> (then)
   n8 = For Each Loop(
    Exec=@n1.then_2
    Array:Array of Light Component Object References=Get Components By Class!(self:Actor Object Reference=Get LightActor(), ComponentClass:Actor Component Class Reference=/Script/Engine.LightComponent) -> (ReturnValue:Array of Light Component Object References)
   ) -> (Array Element:Light Component Object Reference)
   > Set LightA(LightA:Light Component Object Reference=@n8) -> (Output_Get:Light Component Object Reference)
   > n9 = For Each Loop(Array=Get LightComponents() -> (LightComponents:Array of Light Component Object References))
   > Set LightB(LightB:Light Component Object Reference=@n9) -> (Output_Get:Light Component Object Reference)
   > Branch(Condition="Not Equal (Object)"(
     A:Object Reference=Get LightA() -> (LightA:Light Component Object Reference)
     B:Object Reference=Get LightB() -> (LightB:Light Component Object Reference)
    ) -> (ReturnValue:bool=~) {is=title K2Node_PromotableOperator})
   n10 = IsA!(Object:Object Reference=Get LightA(), ObjectClass:Object Class Reference=/Script/Engine.SpotLightComponent) -> (ReturnValue:bool)
   n11 = IsA(Object=Get LightB(), ObjectClass=&SpotLightComponent)
   > n12 = Branch(Condition=AND Boolean(A=@n11, B=@n10))
   > n13 = CheckOverlapSpotToSpot!(
    execute
    self
    SpotLightA:Spot Light Component Object Reference=Cast To SpotLightComponent!(Object=Get LightA()) -> (AsSpot Light Component:Spot Light Component Object Reference, bSuccess:bool=~) {is=title K2Node_DynamicCast}
    SpotLightB:Spot Light Component Object Reference=Cast To SpotLightComponent(Object=Get LightB()) -> (bSuccess=~)
   ) -> (then, bOverlaps:bool) {is=title K2Node_CallFunction}
   > n14 = Branch(Condition=@n13)
   Increment Int(" "=[@n21.then, @n25.then, @n14.then, @n19.then], Value=Get OverlapAmount() -> (OverlapAmount:int=0))
   > AddOverlap!(execute, self:Light Overlap Tool Entry Object Reference=Get EntryWidget(), Overlapping:Light Component Object Reference=Get LightB()) -> (then)
   n15 = Get LightB()
   > AddOverlap(
    self=Find(
     TargetMap:Map of Actor Object References to Light Overlap Tool Entry Object References=Get LightActorsEntries()
     Key:"Actor Object Reference (by ref)"=GetOwner(self=@n15)
    ) -> (Value:Light Overlap Tool Entry Object Reference)
    Overlapping=Get LightA()
   )
   n16 = Branch(execute=@n12.else, Condition=OR Boolean(A=@n11, B=@n10))
   > n17 = Branch(Condition=@n10)
   > n18 = CheckOverlapSpotToPoint!(
    execute
    self
    SpotLight:Spot Light Component Object Reference=Cast To SpotLightComponent(Object=Get LightA()) -> (bSuccess=~)
    PointLight:Point Light Component Object Reference=Cast To PointLightComponent!(Object=Get LightB()) -> (AsPoint Light Component:Point Light Component Object Reference, bSuccess:bool=~) {is=title K2Node_DynamicCast}
   ) -> (then, overlaps:bool) {is=title K2Node_CallFunction}
   > n19 = Branch(Condition=@n18)
   n20 = CheckOverlapSpotToPoint(
    execute=@n17.else
    SpotLight=Cast To SpotLightComponent(Object=Get LightB()) -> (bSuccess=~)
    PointLight=Cast To PointLightComponent(Object=Get LightA()) -> (bSuccess=~)
   )
   > n21 = Branch(Condition=@n20)
   n22 = Get LightB()
   n23 = Get LightA()
   Branch(
    execute=@n16.else
    Condition=AND Boolean(A=IsA(Object=@n22, ObjectClass=/Script/Engine.PointLightComponent), B=IsA(Object=@n23, ObjectClass=&PointLightComponent))
   )
   > n24 = CheckOverlapPointToPoint!(
    execute
    self
    PointLightA:Point Light Component Object Reference=Cast To PointLightComponent(Object=Get LightB()) -> (bSuccess=~)
    PointLightB:Point Light Component Object Reference=Cast To PointLightComponent(Object=Get LightA()) -> (bSuccess=~)
   ) -> (then, bOverlaps:bool) {is=title K2Node_CallFunction}
   > n25 = Branch(Condition=@n24)
  ]}
 EventGraph {nodes [
   "On Clicked (Button_Run)"() {is=title K2Node_ComponentBoundEvent}
   > n1 = Sequence!(execute) -> (then_0, then_1, then_2, then_3:Exec, then_4:Exec)
   n2 = GetAllActorsOfClass!(execute=@n1.then_2, WorldContextObject:Object Reference, ActorClass:Actor Class Reference) -> (then, OutActors:Array of Actor Object References)
   > Set TrackActors(TrackActors:Array of Object References=@n2) -> (Output_Get:Array of Object References)
   n3 = GetAllLevelActors(execute=@n1.then_3, self=Get EditorActorSubsystem())
   > n4 = For Each Loop(Array:Array of Actor Object References=@n3) -> (Array Element:Actor Object Reference)
   n5 = Get Components By Class(self=@n4, ComponentClass=&LightComponent)
   > Append Array!(
    execute
    TargetArray:Array of Light Component Object References=Get LightComponents()
    SourceArray:Array of Light Component Object References=@n5
   ) -> (then)
   > Branch(Condition="integer > integer"(A=Length(TargetArray:Array of Light Component Object References=@n5)))
   > CreateEntry!(execute, self, LightActor:Actor Object Reference=@n4) -> (then) {is=title K2Node_CallFunction}
   ClearChildren(execute=@n1.then_0, self=Get EditorUtilityScrollBox_Entries())
   > Clear(TargetArray:Array of Light Component Object References=Get LightComponents())
   > Clear(TargetArray:Array of Object References=Get TrackActors())
   > Clear!(execute, TargetMap:Map of Actor Object References to Light Overlap Tool Entry Object References=Get LightActorsEntries()) -> (then) {is=title K2Node_CallFunction}
   n6 = Is Valid(exec=@n1.then_1, InputObject=Get ComparerActor())
   n7 = SpawnActorFromClass!(
    execute=@n6.Is Not Valid
    self:Editor Actor Subsystem Object Reference=Get EditorActorSubsystem()
    ActorClass:Actor Class Reference=&/EUBP_LightOverlappingComparer.EUBP_LightOverlappingComparer_C
    Location:Vector="0, 0, 0"
    Rotation:Rotator
    bTransient:bool
   ) -> (then, ReturnValue:EUBP Light Overlapping Comparer Object Reference)
   > Set ComparerActor(ComparerActor:EUBP Light Overlapping Comparer Object Reference=@n7) -> (Output_Get:EUBP Light Overlapping Comparer Object Reference)
   DestroyActor!(
    execute=@n1.then_4
    self:Editor Actor Subsystem Object Reference=Get EditorActorSubsystem()
    ActorToDestroy:Actor Object Reference=Get ComparerActor()
   ) -> (then, ReturnValue:bool)
   Event PreConstruct()
   Event Construct()
   Event Tick()
  ]}
]
variables [
 ComparerActor {type=EUBP_LightOverlappingComparer_C, default=null}
 LightActorsEntries {default {}}
 LightComponents {type="array<LightComponent>", default []}
 TrackActors {type="array<Object>", default []}
]
widgets {
 info {rootWidgetClass=&StackBox, widgetCount=11}
 widgets [
  StackBox_0 {widgetClassPath=&StackBox, widget=StackBox_0}
  VerticalBox_38 {
   widgetClassPath=&VerticalBox
   parent=StackBox_0
   widget=VerticalBox_38
   slot {class=StackBoxSlot, properties {size {sizeRule=Fill, value=1}}}
  }
  Button_Run {widgetClassPath=&EditorUtilityButton, parent=VerticalBox_38, widget=Button_Run, slot {class=VerticalBoxSlot}}
  Text_Button_1 {
   widgetClassPath=&TextBlock
   parent=Button_Run
   widget=Text_Button_1
   properties {
    text=Refresh Light Info
    font {
     bForceMonospaced=false
     bMaterialIsStencil=false
     fontMaterial=null
     fontObject=&Roboto
     letterSpacing=0
     monospacedWidth=1
     size=12
     skewAmount=0
     typefaceFontName=Bold
     outlineSettings {
      bApplyOutlineToDropShadows=false
      bMiteredCorners=false
      bSeparateFillAlpha=false
      outlineMaterial=null
      outlineSize=0
      outlineColor=<rgba 0 0 0 1>
     }
    }
   }
   slot {class=ButtonSlot, properties {padding=<ltrb 4 10 4 10>}}
  }
  Spacer_75 {widgetClassPath=&Spacer, parent=VerticalBox_38, widget=Spacer_75, slot {class=VerticalBoxSlot}}
  EditorUtilityScrollBox_2 {
   widgetClassPath=&EditorUtilityScrollBox
   parent=VerticalBox_38
   widget=EditorUtilityScrollBox_2
   properties {consumeMouseWheel=Never, orientation=Orient_Horizontal}
   slot {class=VerticalBoxSlot, properties {size {sizeRule=Fill, value=1}}}
  }
  VerticalBox_0 {
   widgetClassPath=&VerticalBox
   parent=EditorUtilityScrollBox_2
   widget=VerticalBox_0
   slot {class=ScrollBoxSlot, properties {size {sizeRule=Fill, value=1}}}
  }
  LightOverlapTool_Header {
   widgetClassPath=&/LightOverlapTool_Header
   bIsVariable=true
   parent=VerticalBox_0
   widget=LightOverlapTool_Header
   slot {class=VerticalBoxSlot}
  }
  SizeBox_76 {
   widgetClassPath=&SizeBox
   parent=VerticalBox_0
   widget=SizeBox_76
   slot {class=VerticalBoxSlot, properties {size {sizeRule=Fill, value=1}}}
  }
  Border_61 {
   widgetClassPath=&Border
   parent=SizeBox_76
   widget=Border_61
   properties {brushColor=<rgba 0.01 0.01 0.01 1>}
   slot {class=SizeBoxSlot}
  }
  EditorUtilityScrollBox_Entries {
   widgetClassPath=&EditorUtilityScrollBox
   bIsVariable=true
   parent=Border_61
   widget=EditorUtilityScrollBox_Entries
   slot {class=BorderSlot}
  }
 ]
}
